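# Sound cue for the male WingBat NPC voice line played in the hired DBNO (down but not out) context.
# A Random node picks one of five equally weighted FriendlyDBNO wave variations.
# It uses shared NPC alerted attenuation, the non-critical enemy VO class and henchmen combat chatter concurrency.
# Bus sends feed music, weapon and explosion sidechain buses.
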SoundCue /JethroLager/Assets/Audio/VO/WingBath/Male/Cues/IP_VO_Jethro_WingBat_M_Hired_DBNO_Cue
root=Random_0
nodes [
 {id=Random_0, children [WavePlayer_0,WavePlayer_1,WavePlayer_2,WavePlayer_3,WavePlayer_4], properties {weights [1,1,1,1,1]}}
 {id=WavePlayer_0, properties {soundWaveAssetPtr=/JethroLager/Assets/Audio/VO/WingBath/Male/Waves/IP_VO_Jethro_WingBat_M_FriendlyDBNO_010}}
 {id=WavePlayer_1, properties {soundWaveAssetPtr=/JethroLager/Assets/Audio/VO/WingBath/Male/Waves/IP_VO_Jethro_WingBat_M_FriendlyDBNO_011}}
 {id=WavePlayer_2, properties {soundWaveAssetPtr=/JethroLager/Assets/Audio/VO/WingBath/Male/Waves/IP_VO_Jethro_WingBat_M_FriendlyDBNO_012}}
 {id=WavePlayer_3, properties {soundWaveAssetPtr=/JethroLager/Assets/Audio/VO/WingBath/Male/Waves/IP_VO_Jethro_WingBat_M_FriendlyDBNO_013}}
 {id=WavePlayer_4, properties {soundWaveAssetPtr=/JethroLager/Assets/Audio/VO/WingBath/Male/Waves/IP_VO_Jethro_WingBat_M_FriendlyDBNO_014}}
]
properties {
 attenuationSettings=/Game/Sounds/Fort_Human_VO_BR/NPCs/Shared/NPC_VO_Alerted_Att
 maxDistance=12500
 soundClassObject=/Game/Sounds/FortSoundClassesAndModes/Fort_VO_Enemy_NonCritical
 volumeMultiplier=1.0525
 busSends [
  {audioBus=&AB_Sidechain_FortMusic_Vol, minSendLevel=0, sendLevel=0.25, soundSourceBus=null}
  {audioBus=/Game/Sounds/Modulation/EnvelopeFollowers/Volume/AB_Sidechain_FortWeapon1P_Vol, minSendLevel=0, sendLevel=0.25, soundSourceBus=null}
  {audioBus=&AB_Sidechain_FortWeapon_Vol, minSendLevel=0, sendLevel=0.25, soundSourceBus=null}
  {
   audioBus=&AB_Sidechain_FortExplosion_Vol
   minSendLevel=0
   sendLevel=0.25
   soundSourceBus=null
  }
 ]
 concurrencySet [/Game/Sounds/Fort_Human_VO_BR/NPCs/Shared/NPC_Henchmen_Combat_Chatter_Conc]
}
